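# Prototype translucent, two-sided skeletal-mesh material for the AvenuePole emote props, showing one cell of a temporary texture atlas.
# An enum-driven AvenuePolePropIndex picks among five color-named atlas index parameters (Yellow, Red, Pink, Orange, Blue) through nested Ifs.
# The index, normalized over 64 cells, serves as FlipBook phase to select a static cell in the 8x8 grid.
# The sample supplies color, alpha and a scaled emissive through named reroutes.
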Material /CosmeticCompanions/Assets/Props/Emotes/AvenuePole/Materials/Prototype/M_AvenuePole_Emote_Prop_Atlas_TEMP
graph {
 properties {
  bAutomaticallySetUsageInEditor=false
  bUsedWithSkeletalMesh=true
  blendMode=BLEND_Translucent
  previewMesh=/CosmeticCompanions/Assets/Quadruped/AutoMeow/Meshes/AutoMeow_Quadruped
  twoSided=true
  usedShadingModels=Default Lit
 }
 nodes [
  n1 = NamedRerouteUsage()
  n2 = Constant() {r=64}
  n3 = Constant() {r=8}
  n4 = TextureSampleParameter2D(UVs=MaterialFunctionCall("Animation  Phase (0-1)"=Divide(A=@n1, B=@n2), Number of Rows=@n3, Number of Columns=@n3) {materialFunction=&FlipBook}.UVs) {parameterName=FlipbookTexture, texture=/CosmeticCompanions/Assets/Props/Emotes/AvenuePole/Materials/AvenuePole_Emote_Prop_Atlas_TEMP}
  NamedRerouteDeclaration(None=@n4.A) {name=FlipbookAlpha, nodeColor=<rgba 0 0.8594 1 1>}
  NamedRerouteDeclaration(None=Multiply(
    A=NamedRerouteDeclaration(None=@n4.RGB) {name=FlipbookRGB, nodeColor=<rgba 1 0 0.5391 1>}
    B=ScalarParameter() {defaultValue=0.25, parameterName=EmissiveAmount}
   )) {name=FlipbookEmissive, nodeColor=<rgba 1 0.8203 0 1>}
  n5 = ScalarParameter() {
   controlType=Enumeration
   enumeration=/CosmeticCompanions/Assets/Props/Emotes/AvenuePole/Materials/ENUM_AvenuePolePropIndex
   parameterName=AvenuePolePropIndex
  }
  n6 = ScalarParameter() {defaultValue=63, group=Atlas, parameterName=Orange_Index, sortPriority=3}
  n7 = If(
   A=@n5
   "A > B"=ScalarParameter() {defaultValue=63, group=Atlas, parameterName=Pink_Index, sortPriority=2}
   "A < B"=If(
    A=@n5
    "A > B"=ScalarParameter() {defaultValue=63, group=Atlas, parameterName=Red_Index, sortPriority=1}
    "A < B"=ScalarParameter() {defaultValue=63, group=Atlas, parameterName=Yellow_Index, sortPriority=0}
   ) {constB=0.5}
  ) {constB=1.5}
  NamedRerouteDeclaration(None=If(
    A=@n5
    "A > B"=ScalarParameter() {defaultValue=63, group=Atlas, parameterName=Blue_Index, sortPriority=4}
    "A < B"=If(A=@n5, "A > B"=@n6, "A < B"=@n7) {constB=2.5}
   ) {constB=3.5}) {name=SelectedIndex, nodeColor=<rgba 0.0086 1 0.863 1>}
  NamedRerouteUsage()
  NamedRerouteUsage()
  NamedRerouteUsage()
 ]
}
